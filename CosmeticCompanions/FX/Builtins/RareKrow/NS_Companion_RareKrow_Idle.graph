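# Niagara idle effect for the RareKrow companion that adds fire to the eyes and mouth.
# Socket-tracked mesh and glow-sprite flames sit at eyes and mouth, while a sprite trail sprays curl-noise flames sideways from the mouth in delayed loops.
# User.VariantPicker selects invisible, red-orange or blue variants.
# Eye alpha follows System.EyesAlpha alongside an eyelid-rotation sampling module, and time-of-day curves tint the mesh flames.
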NiagaraSystem /CosmeticCompanions/FX/Builtins/RareKrow/NS_Companion_RareKrow_Idle
dependencies {
 usedDataInterfaces [
  &NiagaraDataInterfaceArrayFloat3
  &NiagaraDataInterfaceColorCurve
  &NiagaraDataInterfaceArrayInt32
  &NiagaraDataInterfaceCurve
  &NiagaraDataInterfaceSpriteRendererInfo
  &NiagaraDataInterfaceSocketReader
 ]
 usedDynamicInputs [
  /Niagara/DynamicInputs/Lerp/SmoothLerpOverTime_Float
  &RemapRange
  &DotProduct
  &ReturnMeshOrientationAxis
 ]
 usedModules [
  &self:SystemSpawnScript.SystemScriptSource.SystemScriptGraph.NiagaraNodeAssignment_0.SetVariables_#1
  &self:SystemSpawnScript.SystemScriptSource.SystemScriptGraph.NiagaraNodeAssignment_3.SetVariables_#2
  &self:SystemSpawnScript.SystemScriptSource.SystemScriptGraph.NiagaraNodeAssignment_2.SetVariables_#3
  &SystemState
  &FN_VariantManager
  &self:SampleEyelidRotation
  &self:SystemSpawnScript.SystemScriptSource.SystemScriptGraph.NiagaraNodeAssignment_4.SetVariables_#4
 ]
}
emitters {
 Eyes_Fire_spline {
  emitterUpdateScript [
   &EmitterState(Inactive Response=::NewEnumerator1, Life Cycle Mode=::NewEnumerator1, Loop Delay=1.25, Loop Duration=1, UseLoopDelay=false)
   &SpawnBurst_Instantaneous(Spawn Count=2, Use Spawn Probability=true, UseLoopCountLimit=false)
  ]
  particleSpawnScript [
   &PartitionParticles()
   &InitializeParticle(
    Color=&MergeHelper_ColorSwitch(
     0=<rgba 0 0 0 0>
     1=<rgba 3 0.05 0 1>
     2=<rgba 0 0.4 1 1>
     3=?
     Index=&MakeIntFromFloat(Float=$User.VariantPicker, Conversion Mode=::NewEnumerator1)
     Num=::NewEnumerator1
    )
    Lifetime=1.25
    Mesh Scale=&Multiply_VectorByFloat(Vector=<0.05 0.05 0.075>, Float=0.9)
    Mesh Scale Mode=::NewEnumerator4
    Sprite Size Mode=::NewEnumerator3
    UsePositionOffset=false
    Write Material Random=true
   )
   &InitialMeshOrientation(
    Mesh Orientation Mode=::NewEnumerator2
    Rotation=&SelectVectorFromArray(
     Direct Array Index=$Particles.PartitionParticles.Partition
     Vector Selection Array=DataInterface() {propertyValues {FloatData [<0 0 0.7>, <0 0 -0.7>]}}
    )
    Use Rotation Vector=true
   )
   &CameraOffset(Camera Offset Amount=1)
  ]
  particleUpdateScript [
   &ParticleState()
   &SocketLocation(
    Interpolate Socket Sampling=true
    SamplePosOffset=true
    Sampled Position Offset=&SelectVectorFromArray(
     Direct Array Index=$Particles.PartitionParticles.Partition
     Vector Selection Array=DataInterface() {propertyValues {FloatData [<1 3 -1>, <-1 3 -1>]}}
    )
    Socket Index=&MergeHelper_IntSwitch(0=1, 1=2, 2=?, 3=?, 4=?, 5=?, Index=$Particles.PartitionParticles.Partition, Num=::NewEnumerator0)
    Socket Reader=$System.Socket Reader
    Socket Sampling Mode=::NewEnumerator3
    VelocitySpeedLimit=false
   )
   &ScaleColor(
    Scale Alpha=&Multiply_Float(
     A=&Multiply_Float(
      A=0.25
      B=&MergeHelper_FloatSwitch(
       0=0
       1=1
       2=1
       3=?
       Index=&MakeIntFromFloat(Float=$User.VariantPicker, Conversion Mode=::NewEnumerator1)
       Num=::NewEnumerator1
      )
     )
     B=$System.EyesAlpha
    )
    Scale RGB=&MakeVectorFromLinearColorRGB(Color=&ColorFromCurve(
      ColorCurve=DataInterface() {propertyValues {
        AlphaCurve {keys [<at 0 1>,<at 0.125 1>,<at 0.25 1>,<at 0.375 1>,<at 0.5 1>,<at 0.625 1>,<at 0.75 1>,<at 0.875 1>,<at 1 1>]}
        BlueCurve {keys [<at 0.0936 0.5104>,<at 0.166 0.6172>,<at 0.208 0.5395>,<at 0.25 0.6172>,<at 0.291 0.6939>,<at 0.333 0.7379>,<at 0.375 0.7835>,<at 0.416 0.7991>,<at 0.458 0.8148>,<at 0.5 0.8308>,<at 0.541 0.8148>,<at 0.583 0.7991>,<at 0.625 0.7991>,<at 0.666 0.7529>,<at 0.708 0.7084>,<at 0.75 0.5711>,<at 0.791 0.4735>,<at 0.833 0.3231>,<at 0.909 0.5104>]}
        GreenCurve {keys [<at 0.0936 0.3369>,<at 0.166 0.5333>,<at 0.208 0.5271>,<at 0.25 0.6376>,<at 0.291 0.7529>,<at 0.333 0.7835>,<at 0.375 0.8148>,<at 0.416 0.7991>,<at 0.458 0.8148>,<at 0.5 0.8469>,<at 0.541 0.7991>,<at 0.583 0.7991>,<at 0.625 0.8148>,<at 0.666 0.7991>,<at 0.708 0.7529>,<at 0.75 0.5711>,<at 0.791 0.5029>,<at 0.833 0.5029>,<at 0.909 0.3369>]}
        RedCurve {keys [<at 0.0936 0.1249>,<at 0.166 0.7084>,<at 0.208 0.6867>,<at 0.25 0.8148>,<at 0.291 0.8879>,<at 0.333 0.8963>,<at 0.375 0.9131>,<at 0.416 0.8714>,<at 0.458 0.8714>,<at 0.5 0.9131>,<at 0.541 0.8632>,<at 0.583 0.8714>,<at 0.625 0.9131>,<at 0.666 0.8963>,<at 0.708 0.8714>,<at 0.75 0.7529>,<at 0.791 0.7529>,<at 0.833 0.8469>,<at 0.909 0.1249>]}
       }}
      CurveIndex=Linked() {linkedVariable {name=NPC.FortniteNPC.TimeOfDay_0-1, type=NiagaraFloat}}
     ))
    ScaleA=true
    ScaleRGB=true
   )
   &DynamicMaterialParameters(
    Index 0 Param 1=15
    Index 0 Param 2=2
    Index 0 Param 3=&Multiply_Float(A=$System.Age, B=&RandomRangeFloat(Minimum=-1.1, Maximum=-1.2))
    Index 0 Param 4=&RandomRangeFloat(Minimum=1.345, Maximum=2)
    Module.Index0Param0Write=true
    Module.Index0Param1Write=true
    Module.Index0Param2Write=true
    Module.Index0Param3Write=true
   )
   /SetVariables(Particles.CustomDynParam_0:Vector4f=&MakeVector4(
     X=5
     Y=3
     Z=&Add_Float(
      A=&Multiply_Float(A=$System.Age, B=&RandomRangeFloat(Minimum=-1.05, Maximum=-1.2))
      B=&RandomRangeFloat(Minimum=0, Maximum=100)
     )
     W=&RandomRangeFloat(Minimum=1.4, Maximum=2)
    )) {isSetParameters=true, module=SetVariables_#5}
   /SetVariables(
    Particles.GlowCamOffset=2
    Particles.GlowColor=&MakeLinearColorFromVectorAndFloat(
     "Vector (RGB)"=&MakeVectorFromLinearColorRGB(Color=&MergeHelper_ColorSwitch(
       0=<rgba 0 0 0 0>
       1=<rgba 2 0.1322 0 1>
       2=<rgba 0 0.6609 1 1>
       3=?
       Index=&MakeIntFromFloat(Float=$User.VariantPicker)
       Num=::NewEnumerator1
      ))
     "Float (Alpha)"=&Multiply_Float(
      A=$System.EyesAlpha
      B=&MergeHelper_FloatSwitch(
       0=0
       1=0.33
       2=0.33
       3=?
       Index=&MakeIntFromFloat(Float=$User.VariantPicker, Conversion Mode=::NewEnumerator1)
       Num=::NewEnumerator1
      )
     )
    )
   ) {isSetParameters=true, module=SetVariables_#6}
  ]
  rendererClasses [&NiagaraMeshRendererProperties, &NiagaraSpriteRendererProperties]
  renderers [
   {rendererClass=&NiagaraMeshRendererProperties, rendererIndex=0}
   {rendererClass=&NiagaraMeshRendererProperties, rendererIndex=1}
   {rendererClass=&NiagaraSpriteRendererProperties, rendererIndex=2}
  ]
 }
 Mouth_Fire_Trail {
  emitterUpdateScript [
   &EmitterState(Life Cycle Mode=::NewEnumerator1, Loop Delay=1.25, Loop Duration=1, UseLoopDelay=true)
   &SpawnRate(SpawnRate=25, Use Spawn Probability=true)
  ]
  particleSpawnScript [
   /SetVariables(Particles.Random size:NiagaraFloat=&RandomRangeFloat(Minimum=0.3, Maximum=0.7)) {isSetParameters=true}
   &PartitionParticles(Number Of Partitions=4)
   &InitializeParticle(
    Color=&MergeHelper_ColorSwitch(
     0=<rgba 0 0 0 0>
     1=<rgba 3 0.1 0 2>
     2=<rgba 0 1.8261 3 2>
     3=?
     Index=&MakeIntFromFloat(Float=$User.VariantPicker)
     Num=::NewEnumerator1
    )
    Lifetime=&RandomRangeFloat(Minimum=0.45, Maximum=0.5)
    Sprite Rotation Mode=::NewEnumerator2
    Sprite Size Mode=::NewEnumerator3
    Sprite UV Mode=::NewEnumerator3
    Uniform Sprite Size=&Multiply_Float(A=25, B=$Particles.Random size)
    UsePositionOffset=false
    Write Material Random=true
   ) {module=InitializeParticle001}
   &DynamicMaterialParameters(
    Index 0 Param 1=5
    Index 0 Param 2=&RandomRangeFloat(Minimum=0, Maximum=4)
    Index 0 Param 4=2.5
    Module.Param0WriteEnabled=true
    Module.Param1WriteEnabled=true
    Module.Param2WriteEnabled=false
    Module.Param3WriteEnabled=true
   ) {module=DynamicMaterialParameters001}
   /SetVariables() {isSetParameters=true}
   /SetVariables(Particles.DY3:Vector4f=&MakeVector4(X=3, Y=1, Z=1, W=1)) {isSetParameters=true}
   &SocketLocation(
    Interpolate Socket Sampling=true
    SamplePosOffset=true
    Sampled Position Offset=&SelectVectorFromArray(
     Direct Array Index=$Particles.PartitionParticles.Partition
     Vector Selection Array=DataInterface() {propertyValues {FloatData [<4 20 -0.7>, <8 18 0.8>, <-4 20 -0.7>, <-8 18 0.8>]}}
    )
    Socket Reader=$System.Socket Reader
    Socket Sampling Mode=::NewEnumerator3
    VelocitySpeedLimit=false
   )
   &AddVelocity(
    Velocity=&SelectVectorFromArray(
     Direct Array Index=$Particles.PartitionParticles.Partition
     Vector Selection Array=DataInterface() {propertyValues {FloatData [<50 0 50>, <50 0 50>, <-50 0 50>, <-50 0 50>]}}
    )
    Velocity Speed Scale=&RandomRangeFloat(Minimum=0.5, Maximum=0.7)
   )
   &CameraOffset(Camera Offset Amount=5)
  ]
  particleUpdateScript [
   /SetVariables(Particles.Size Scale:NiagaraFloat=&FloatFromCurve(
     FloatCurve=DataInterface() {propertyValues {Curve {keys [{interpMode=RCIM_Cubic, tangentMode=RCTM_User, time=0, value=0.35}, <at 0.2 0.7>]}}}
     Scale Curve=0.7
    )) {isSetParameters=true}
   &SubUVAnimation(
    End Frame Range Override=3
    Evaluation Type=::NewEnumerator0
    "Sprite Renderer "=DataInterface() {propertyValues {SpriteRenderer=&self:Mouth_Fire_Trail_0.NiagaraSpriteRendererProperties_1}}
    SubUV Animation Mode:ENiagara_SubUVLookupModeV2=::NewEnumerator2
    UseEndFrame=true
    UseStartFrame=true
   )
   /Niagara/Modules/Update/Forces/CurlNoiseForce(Noise Frequency=100, Noise Strength=&Multiply_Float(A=5, B=10), Pan Noise=true)
   &ParticleState()
   &ScaleSpriteSize(Scale Sprite Size Mode=::NewEnumerator0, Uniform Scale Factor=$Particles.Size Scale) {module=ScaleSpriteSize001}
   &SpriteRotationRate(Rotation Rate=&RandomRangeFloat(Minimum=-400, Maximum=400)) {module=SpriteRotationRate001}
   &Drag(Drag=&RandomRangeFloat(Minimum=2, Maximum=3), Use Linear Drag=true, Use Rotational Drag=true)
   &SolveForcesAndVelocity(Clamp Velocity=false, Limit Acceleration=false)
   &FN_Sprite_Attributes(
    Color=&MakeLinearColorFromVectorAndFloat(
     "Vector (RGB)"=&MergeHelper_VectorSwitch(
      0=<0 0 0>
      1=&MakeVectorFromLinearColorRGB(Color=&ColorFromCurve(ColorCurve=DataInterface() {propertyValues {
          AlphaCurve {keys []}
          BlueCurve {keys [<at 0 0.0783 RCIM_Cubic>,<at 0.5807 0.0166>]}
          GreenCurve {keys [<at 0 0.0783 RCIM_Cubic>,<at 0.5807 0>]}
          RedCurve {keys [<at 0 3 RCIM_Cubic>,<at 0.5807 0.3177>]}
         }}))
      2=&MakeVectorFromLinearColorRGB(Color=&ColorFromCurve(ColorCurve=DataInterface() {propertyValues {
          AlphaCurve {keys []}
          BlueCurve {keys [<at 0 3 RCIM_Cubic>,<at 0.5807 0.5>]}
          GreenCurve {keys [<at 0 0.6624 RCIM_Cubic>,<at 0.5807 0.0696>]}
          RedCurve {keys [<at 0 0.1094 RCIM_Cubic>,<at 0.5807 0>]}
         }}))
      3=?
      Index=&MakeIntFromFloat(Float=$User.VariantPicker)
      Num=::NewEnumerator1
     )
     "Float (Alpha)"=&Multiply_Float(
      A=&FloatFromCurve(
       FloatCurve=DataInterface() {propertyValues {Curve {keys [
           {tangentMode=RCTM_User, time=0, value=1}
           {arriveTangent=-0.0405, interpMode=RCIM_Cubic, leaveTangent=-0.0405, tangentMode=RCTM_User, time=0.3, value=1}
           {arriveTangent=-1.1636, interpMode=RCIM_Cubic, leaveTangent=-1.1636, tangentMode=RCTM_User, time=1, value=0}
          ]}}}
       Scale Curve=&MergeHelper_FloatSwitch(
        0=0
        1=1
        2=1
        3=?
        Index=&MakeIntFromFloat(Float=$User.VariantPicker, Conversion Mode=::NewEnumerator1)
        Num=::NewEnumerator1
       )
      )
      B=0.75
     )
    )
    Create Color Attribute=true
    Create SpriteSize Attribute=true
    SpriteSize=&Vector2DFromFloat(Value=&Multiply_Float(A=33, B=&Multiply_Float(A=$Particles.Random size, B=$Particles.Size Scale)))
    Write to Module Namespace=true
   )
   &FN_Sprite_Attributes(
    Color=&MakeLinearColorFromVectorAndFloat(
     "Vector (RGB)"=&MakeVectorFromLinearColorRGB(Color=<rgba 1.5 0 0 0.5>)
     "Float (Alpha)"=&Multiply_Float(
      A=&FN_FrontEndFloatModifier(Gameplay=0.2, Front End=0.2)
      B=&MergeHelper_FloatSwitch(
       0=0
       1=0.005
       2=0.005
       3=?
       Index=&MakeIntFromFloat(Float=$User.VariantPicker, Conversion Mode=::NewEnumerator1)
       Num=::NewEnumerator1
      )
     )
    )
    Create Color Attribute=true
    Create SpriteSize Attribute=true
    SpriteSize=&Vector2DFromFloat(Value=&Multiply_Float(A=45, B=&Multiply_Float(A=$Particles.Random size, B=$Particles.Size Scale)))
    Write to Module Namespace=true
   ) {module=FN_Sprite_Attributes001}
  ]
  rendererClasses [&NiagaraSpriteRendererProperties]
  renderers [
   {rendererClass=&NiagaraSpriteRendererProperties}
   {rendererClass=&NiagaraSpriteRendererProperties, rendererIndex=1}
  ]
 }
 Mouth_Fire_spline {
  emitterUpdateScript [
   &EmitterState(Inactive Response=::NewEnumerator1, Life Cycle Mode=::NewEnumerator1, Loop Delay=1.25, Loop Duration=1, UseLoopDelay=true)
   &SpawnBurst_Instantaneous(Spawn Count=4, Use Spawn Probability=true, UseLoopCountLimit=false)
  ]
  particleSpawnScript [
   &PartitionParticles(Number Of Partitions=4)
   &InitializeParticle(
    Color=&MergeHelper_ColorSwitch(
     0=<rgba 0 0 0 0>
     1=<rgba 1 0.05 0 1>
     2=<rgba 0 0.2957 1 1>
     3=?
     Index=&MakeIntFromFloat(Float=$User.VariantPicker)
     Num=::NewEnumerator1
    )
    Lifetime=1.25
    Mesh Scale=&Multiply_VectorByFloat(
     Vector=<0.05 0.05 0.075>
     Float=&MergeHelper_FloatSwitch(0=1, 1=0.6, 2=1, 3=0.6, Index=$Particles.PartitionParticles.Partition, Num=::NewEnumerator2)
    )
    Mesh Scale Mode=::NewEnumerator4
    Sprite Size Mode=::NewEnumerator3
    UsePositionOffset=false
    Write Material Random=true
   )
   &InitialMeshOrientation(
    Mesh Orientation Mode=::NewEnumerator2
    Rotation=&SelectVectorFromArray(
     Direct Array Index=$Particles.PartitionParticles.Partition
     Vector Selection Array=DataInterface() {propertyValues {FloatData [<0 -0.05 -0.33>, <0 -0.05 -0.33>, <0 0.05 0.33>, <0 0.05 0.33>]}}
    )
    Use Rotation Vector=true
   )
   &CameraOffset(Camera Offset Amount=1)
  ]
  particleUpdateScript [
   &ParticleState()
   &SocketLocation(
    Interpolate Socket Sampling=true
    SamplePosOffset=true
    Sampled Position Offset=<0 0 -1>
    Socket Index=&SelectIntFromArray(Direct Array Index=$Particles.PartitionParticles.Partition, Int Selection Array=DataInterface() {propertyValues {IntData [3,4,5,6]}})
    Socket Reader=$System.Socket Reader
    Socket Sampling Mode=::NewEnumerator3
    VelocitySpeedLimit=false
   )
   &ScaleMeshSize(Scale Factor=&Multiply_VectorByFloat(
     Vector=<1 1 1>
     Float=&FloatFromCurve(FloatCurve=DataInterface() {propertyValues {Curve {
         defaultValue=1
         keys [
          {arriveTangent=9.0725, interpMode=RCIM_Cubic, leaveTangent=9.0725, tangentMode=RCTM_User, time=0, value=0}
          {arriveTangent=0.3838, interpMode=RCIM_Cubic, leaveTangent=0.3838, tangentMode=RCTM_User, time=0.1408, value=0.9847}
          {arriveTangent=-0.0405, interpMode=RCIM_Cubic, leaveTangent=-0.0405, tangentMode=RCTM_User, time=1.0122, value=0.5567}
         ]
        }}})
    ))
   &ScaleColor(
    Scale Alpha=&Multiply_Float(
     A=0.25
     B=&FloatFromCurve(
      FloatCurve=DataInterface() {propertyValues {Curve {
         defaultValue=1
         keys [
          {arriveTangent=9.0725, interpMode=RCIM_Cubic, leaveTangent=9.0725, tangentMode=RCTM_User, time=0, value=0}
          {arriveTangent=-2.5249, interpMode=RCIM_Cubic, leaveTangent=-2.5249, tangentMode=RCTM_User, time=0.1408, value=0.9847}
          {arriveTangent=-0.0405, interpMode=RCIM_Cubic, leaveTangent=-0.0405, tangentMode=RCTM_User, time=1, value=0}
         ]
        }}}
      Scale Curve=&MergeHelper_FloatSwitch(
       0=0
       1=1
       2=1
       3=?
       Index=&MakeIntFromFloat(Float=$User.VariantPicker, Conversion Mode=::NewEnumerator1)
       Num=::NewEnumerator1
      )
     )
    )
    Scale RGB=&MakeVectorFromLinearColorRGB(Color=&ColorFromCurve(
      ColorCurve=DataInterface() {propertyValues {
        AlphaCurve {keys [<at 0 1>,<at 0.125 1>,<at 0.25 1>,<at 0.375 1>,<at 0.5 1>,<at 0.625 1>,<at 0.75 1>,<at 0.875 1>,<at 1 1>]}
        BlueCurve {keys [<at 0.0936 0.5104>,<at 0.166 0.6172>,<at 0.208 0.5395>,<at 0.25 0.6172>,<at 0.291 0.6939>,<at 0.333 0.7379>,<at 0.375 0.7835>,<at 0.416 0.7991>,<at 0.458 0.8148>,<at 0.5 0.8308>,<at 0.541 0.8148>,<at 0.583 0.7991>,<at 0.625 0.7991>,<at 0.666 0.7529>,<at 0.708 0.7084>,<at 0.75 0.5711>,<at 0.791 0.4735>,<at 0.833 0.3231>,<at 0.909 0.5104>]}
        GreenCurve {keys [<at 0.0936 0.3369>,<at 0.166 0.5333>,<at 0.208 0.5271>,<at 0.25 0.6376>,<at 0.291 0.7529>,<at 0.333 0.7835>,<at 0.375 0.8148>,<at 0.416 0.7991>,<at 0.458 0.8148>,<at 0.5 0.8469>,<at 0.541 0.7991>,<at 0.583 0.7991>,<at 0.625 0.8148>,<at 0.666 0.7991>,<at 0.708 0.7529>,<at 0.75 0.5711>,<at 0.791 0.5029>,<at 0.833 0.5029>,<at 0.909 0.3369>]}
        RedCurve {keys [<at 0.0936 0.1249>,<at 0.166 0.7084>,<at 0.208 0.6867>,<at 0.25 0.8148>,<at 0.291 0.8879>,<at 0.333 0.8963>,<at 0.375 0.9131>,<at 0.416 0.8714>,<at 0.458 0.8714>,<at 0.5 0.9131>,<at 0.541 0.8632>,<at 0.583 0.8714>,<at 0.625 0.9131>,<at 0.666 0.8963>,<at 0.708 0.8714>,<at 0.75 0.7529>,<at 0.791 0.7529>,<at 0.833 0.8469>,<at 0.909 0.1249>]}
       }}
      CurveIndex=Linked() {linkedVariable {name=NPC.FortniteNPC.TimeOfDay_0-1, type=NiagaraFloat}}
     ))
    ScaleA=true
    ScaleRGB=true
   )
   &DynamicMaterialParameters(
    Index 0 Param 1=15
    Index 0 Param 3=&Multiply_Float(A=$System.Age, B=&RandomRangeFloat(Minimum=-1.1, Maximum=-1.2))
    Index 0 Param 4=&RandomRangeFloat(Minimum=1.345, Maximum=2)
    Module.Index0Param0Write=true
    Module.Index0Param1Write=true
    Module.Index0Param2Write=true
    Module.Index0Param3Write=true
   )
   /SetVariables(Particles.CustomDynParam_0:Vector4f=&MakeVector4(
     X=5
     Y=3
     Z=&Add_Float(
      A=&Multiply_Float(A=$System.Age, B=&RandomRangeFloat(Minimum=-1.05, Maximum=-1.2))
      B=&RandomRangeFloat(Minimum=0, Maximum=100)
     )
     W=&RandomRangeFloat(Minimum=1.4, Maximum=2)
    )) {isSetParameters=true, module=SetVariables_#5}
   /SetVariables(
    Particles.GlowCamOffset=2
    Particles.GlowColor=&MakeLinearColorFromVectorAndFloat(
     "Vector (RGB)"=&MakeVectorFromLinearColorRGB(Color=&MergeHelper_ColorSwitch(
       0=<rgba 0 0 0 0>
       1=<rgba 1 0.1565 0 1>
       2=<rgba 0 0.7131 1 1>
       3=?
       Index=&MakeIntFromFloat(Float=$User.VariantPicker, Conversion Mode=::NewEnumerator1)
       Num=::NewEnumerator1
      ))
     "Float (Alpha)"=&FloatFromCurve(
      FloatCurve=DataInterface() {propertyValues {Curve {
         defaultValue=1
         keys [
          {interpMode=RCIM_Cubic, tangentMode=RCTM_User, time=0, value=0}
          {arriveTangent=-0.0158, interpMode=RCIM_Cubic, leaveTangent=-0.0158, tangentMode=RCTM_User, time=0.1283, value=0.9806}
          {arriveTangent=-1.4591, interpMode=RCIM_Cubic, leaveTangent=-1.4591, tangentMode=RCTM_User, time=0.7028, value=0.5923}
          {tangentMode=RCTM_User, time=1, value=0}
         ]
        }}}
      Scale Curve=&MergeHelper_FloatSwitch(
       0=0
       1=0.33
       2=0.33
       3=?
       Index=&MakeIntFromFloat(Float=$User.VariantPicker, Conversion Mode=::NewEnumerator1)
       Num=::NewEnumerator1
      )
     )
    )
   ) {isSetParameters=true, module=SetVariables_#6}
  ]
  rendererClasses [&NiagaraMeshRendererProperties, &NiagaraSpriteRendererProperties]
  renderers [
   {rendererClass=&NiagaraMeshRendererProperties, rendererIndex=0}
   {rendererClass=&NiagaraMeshRendererProperties, rendererIndex=1}
   {rendererClass=&NiagaraSpriteRendererProperties, rendererIndex=2}
  ]
 }
}
summary {
 systemName=NS_Companion_RareKrow_Idle
 emitters [
  {
   emitterName=Eyes_Fire_spline
   rendererClasses [&NiagaraMeshRendererProperties, &NiagaraSpriteRendererProperties]
  }
  {
   emitterName=Mouth_Fire_spline
   rendererClasses [&NiagaraMeshRendererProperties, &NiagaraSpriteRendererProperties]
  }
  {emitterName=Mouth_Fire_Trail, rendererClasses [&NiagaraSpriteRendererProperties]}
 ]
 userVariables [
  User.bIsAds {type=NiagaraBool}
  User.bIsFrontEnd {type=NiagaraBool}
  User.bIsFrontEndPreview {type=NiagaraBool}
  User.VariantPicker {type=NiagaraFloat, defaultValue=0, description=""}
 ]
}
